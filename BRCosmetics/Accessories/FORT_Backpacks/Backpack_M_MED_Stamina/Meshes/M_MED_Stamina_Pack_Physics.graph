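# Physics asset for the M_MED_Stamina backpack cosmetic, defining a short jointed rope chain alongside capsule collision for the pack.
# Three dyn_rope capsule bodies chain from the small rope_1 body, while three larger capsules on the pack bone approximate its volume.
# Each joint has zero twist and a much wider second swing limit than first, so the rope bends mainly in one plane.
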PhysicsAsset /BRCosmetics/Accessories/FORT_Backpacks/Backpack_M_MED_Stamina/Meshes/M_MED_Stamina_Pack_Physics
bodies {
 dyn_rope_2 [{shapeType=Capsule, length=4.8193, radius=1, shapeName=dyn_rope_2_capsule, center=<3.5 0 0>, rotation=<rot 90 0 0>}]
 dyn_rope_3 [{shapeType=Capsule, length=4.8193, radius=1, shapeName=dyn_rope_3_capsule, center=<3.5 0 0>, rotation=<rot 90 0 0>}]
 dyn_rope_4 [{shapeType=Capsule, length=1, radius=1, shapeName=dyn_rope_4_capsule, center=<2 0 0>, rotation=<rot 90 0 0>}]
 pack [
  {
   shapeType=Capsule
   length=68.4604
   radius=2.9508
   shapeName=pack_capsule
   center=<0.3226 -2.2502 -0.5547>
   rotation=<rot 1.7484 -88.9795 -38.135>
  }
  {
   shapeType=Capsule
   length=20.6692
   radius=18.6235
   shapeName=pack_capsule
   center=<0 13.9682 -29.2392>
   rotation=<rot -90 179.9802 -179.9803>
  }
  {
   shapeType=Capsule
   length=19.0812
   radius=33.2233
   shapeName=pack_capsule
   center=<-1.719 29.5723 -14.2781>
   rotation=<rot -85.0005 -90 90.0001>
  }
 ]
 rope_1 [{shapeType=Capsule, length=0.25, radius=0.25, shapeName=rope_1_capsule, center=<0.5 0 0>, rotation=<rot 90 0 0>}]
}
constraints [
 {bone1Name=dyn_rope_2, bone2Name=rope_1, swing1LimitDegrees=1, swing2LimitDegrees=35, twistLimitDegrees=0}
 {bone1Name=dyn_rope_3, bone2Name=dyn_rope_2, swing1LimitDegrees=10, swing2LimitDegrees=50, twistLimitDegrees=0}
 {bone1Name=dyn_rope_4, bone2Name=dyn_rope_3, swing1LimitDegrees=1, swing2LimitDegrees=25, twistLimitDegrees=0}
]
